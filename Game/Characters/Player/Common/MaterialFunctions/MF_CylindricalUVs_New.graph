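# Material function that builds cylindrical UVs around the Z axis from a position input, defaulting to pre-skinned local position.
# U is the signed XY angle over 2π times Circumference Tiling (whole numbers hide the seam), using fast or HQ arccosine via static switch.
# V is Z divided by a height input, defaulting to pre-skinned bounds Z extent.
# Explicit DDX/DDY outputs derive from view-right projected position and V, independent of the wrapped U.
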MaterialFunction /Game/Characters/Player/Common/MaterialFunctions/MF_CylindricalUVs_New
graph {nodes [
  n1 = FunctionInput() {
   bUsePreviewValueAsDefault=true
   description=Use whole numbers to prevent seam
   inputName=Circumference Tiling
   inputType=FunctionInput_Scalar
   previewValue=<2 0 0 1>
  }
  n2 = Constant3Vector() {constant=<rgba 1 1 0 0>}
  n3 = FunctionInput(Preview=VertexInterpolator(VS=PreSkinnedPosition()).PS) {bUsePreviewValueAsDefault=true, description=Defaults to Preskinned Local Position, inputName=Position}
  n4 = ComponentMask(None=CrossProduct(A=Multiply(A=@n2, B=@n3), B=Constant3Vector() {constant=<rgba 0 0 1 0>})) {r=true}
  n5 = DotProduct(A=Normalize(VectorInput=ComponentMask(None=@n3) {g=true, r=true}), B=Constant2Vector() {r=1})
  n6 = Arccosine!(None=@n5)
  n7 = ArccosineFast!(None=@n5)
  n8 = FunctionInput(Preview=StaticBool()) {
   bUsePreviewValueAsDefault=true
   description=Uses Arccosine instead of ArccosineFast for U value calculation
   inputName=HQ Angle Calculation?
   inputType=FunctionInput_StaticBool
   sortPriority=100
  }
  n9 = Divide(
   A=Multiply(A=Sign(None=@n4), B=StaticSwitch(True=@n6, False=@n7, Value=@n8))
   B=MaterialFunctionCall(Multiplier=Constant() {r=2}) {materialFunction=&Pi}.Result
  )
  n10 = FunctionInput(Preview=ComponentMask(None=PreSkinnedLocalBounds().Extents) {b=true}) {
   bUsePreviewValueAsDefault=true
   description=Defaults to Preskinned Local Bounds Z
   inputName=Z Height
   inputType=FunctionInput_Scalar
   sortPriority=10
  }
  n11 = Divide(A=@n3, B=@n10)
  n12 = ComponentMask(None=@n11) {b=true}
  FunctionOutput(None=AppendVector(A=Multiply(A=@n1, B=@n9), B=@n12)) {outputName=UVs}
  n13 = Normalize(VectorInput=Transform(None=Constant3Vector() {constant=<rgba 1 0 0 0>}) {transformSourceType=TRANSFORMSOURCE_View})
  FunctionOutput(None=DDX(Value=DotProduct(A=@n11, B=@n13))) {outputName=Explicit DDX, sortPriority=5}
  FunctionOutput(None=DDY(Value=@n12)) {outputName=Explicit DDY, sortPriority=10}
 ]}
